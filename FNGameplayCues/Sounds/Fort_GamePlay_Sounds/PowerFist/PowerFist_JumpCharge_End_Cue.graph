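# Sound cue for the end of the PowerFist jump charge, judging by its name.
# A single wave player, reusing a Neon rocket launcher close-fire wave, feeds a default Modulator for slight pitch and volume variation.
# It uses the PowerFist close attenuation asset and a third-person ranged weapon fire sound class at reduced volume.
# Overridden concurrency allows one instance, stopping the farthest, then preventing new ones.
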SoundCue /FNGameplayCues/Sounds/Fort_GamePlay_Sounds/PowerFist/PowerFist_JumpCharge_End_Cue
root=Modulator_1
nodes [
 {id=Modulator_1, children [WavePlayer_17]}
 {id=WavePlayer_17, properties {soundWaveAssetPtr=/FNGameplayCues/Sounds/Fort_Weapons/Launchers/Neon/Neon_RocketLauncher_Fire_Close_01}}
]
properties {
 attenuationSettings=/FNGameplayCues/Sounds/Attenuation/Fort_Weapon_Powerfist_Close_Athena_Att
 bOverrideConcurrency=true
 maxDistance=3584
 soundClassObject=&Fort_WeaponRanged_Fire_3P_PSM
 volumeMultiplier=0.65
 attenuationOverrides {falloffDistance=5000, attenuationShapeExtents {x=1000}}
 concurrencyOverrides {maxCount=1, resolutionRule=StopFarthestThenPreventNew}
}
