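# Sound cue for the PowerFist jump charge at full, built from borrowed Neon rocket launcher waves.
# A mixer layers the environment impact wave at half level with the charge-stage-reached wave at full level, under a reduced cue volume.
# It uses close PowerFist attenuation and the third-person ranged weapon fire sound class.
# A concurrency override allows one instance, stopping the farthest and then preventing new ones.
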SoundCue /FNGameplayCues/Sounds/Fort_GamePlay_Sounds/PowerFist/PowerFist_JumpCharge_Full_Cue
root=Mixer_0
nodes [
 {id=Mixer_0, children [WavePlayer_17, WavePlayer_0], properties {inputVolume [0.5,1]}}
 {id=WavePlayer_0, properties {soundWaveAssetPtr=/FNGameplayCues/Sounds/Fort_Weapons/Launchers/Neon/Neon_Rocket_ChargeStageReached}}
 {id=WavePlayer_17, properties {soundWaveAssetPtr=/FNGameplayCues/Sounds/Fort_Weapons/Launchers/Neon/Impact_NeonRocket_Env_01}}
]
properties {
 attenuationSettings=/FNGameplayCues/Sounds/Attenuation/Fort_Weapon_Powerfist_Close_Athena_Att
 bOverrideConcurrency=true
 maxDistance=3584
 soundClassObject=&Fort_WeaponRanged_Fire_3P_PSM
 volumeMultiplier=0.25
 attenuationOverrides {falloffDistance=5000, attenuationShapeExtents {x=1000}}
 concurrencyOverrides {maxCount=1, resolutionRule=StopFarthestThenPreventNew}
}
